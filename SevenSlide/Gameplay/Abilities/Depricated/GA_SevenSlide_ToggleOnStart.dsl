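; Gameplay ability, in a Depricated folder, that applies a starting gameplay effect to its owner when activated; the name suggests a toggle-on-start state.
; On activation it applies the effect held in the StartingGE variable to the owner.
; The end-ability event has no logic, and the graph never ends the ability or removes the effect itself.
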(event Ability|EventActivateAbility
  (Ability|ApplyGameplayEffectToOwner (Variables|Default|GetStartingGE)))

(event Ability|EventOnEndAbility
  (params (bWasCancelled :type Boolean)))
