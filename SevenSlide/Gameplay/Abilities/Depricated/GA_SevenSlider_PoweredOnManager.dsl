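; Deprecated-folder gameplay ability that manages the SevenSlider gadget's powered-on state and matching cosmetic parts.
; On activation it applies the equipped character-part effect, briefly delays a cosmetic reselect, then waits for toggle-power events.
; Each event checks the equipped tag: power-off swaps effects immediately, while power-on applies the equipped effect and delays the character-part swap by timer.
; Toggles reselect standard and Mythic gadget cosmetics; ending clears the wait task and swap timer.
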(event Custom|SelectGadget
  (Class|FortCosmeticFunctionLibrary|ReselectGadgetCosmeticParts (CallFunction|GetAvatarasFortPlayerPawn) "/SevenSlide/Gameplay/AGID_BoostSevenSlider.AGID_BoostSevenSlider"))

(event Ability|EventActivateAbility
  (bind _returnvalue (CallFunction|GetAvatarasFortPlayerPawn))
  (bind _ge_equipped (Variables|Default|GetGE_Equipped))
  ; Activate Ability - start toggled on
  (Ability|ApplyGameplayEffectToOwner (Variables|Default|GetGE_Equipped_CharacterPart))
  (Utilities|Time|SetTimerbyEvent (delegate Custom|SelectGadget) 0.01)
  (bind _asynctaskproxy (Ability|Tasks|WaitGameplayEvent "(TagName=\"sevenslide.TogglePowerEvent\")")
    (:then
      (Variables|Default|SetTogglePowerTask _asynctaskproxy))
    (:EventReceived
      (if (GameplayTags|HasMatchingGameplayTag (CallFunction|GetAvatarasFortPlayerPawn) (Variables|Default|GetEquippedTag))
        (GameplayEffects|RemoveActiveGameplayEffectbySourceEffect (Ability|GetAbilitySystemComponent (CallFunction|GetAvatarasFortPlayerPawn)) _ge_equipped 0 -1)
        (GameplayEffects|RemoveActiveGameplayEffectbySourceEffect (Ability|GetAbilitySystemComponent (CallFunction|GetAvatarasFortPlayerPawn)) (Variables|Default|GetGE_Equipped_CharacterPart) 0 -1)
        (Ability|ApplyGameplayEffectToOwner (Variables|Default|GetGE_Not_Equipped_CharacterPart))
        (Class|FortCosmeticFunctionLibrary|ReselectGadgetCosmeticParts _returnvalue "/SevenSlide/Gameplay/AGID_BoostSevenSlider.AGID_BoostSevenSlider")
        (Class|FortCosmeticFunctionLibrary|ReselectGadgetCosmeticParts _returnvalue "/SevenSlide/Gameplay/AGID_Mythic_BoostSevenSlider.AGID_Mythic_BoostSevenSlider")
        (else
          (Ability|ApplyGameplayEffectToOwner _ge_equipped)
          (bind _returnvalue_2 (Utilities|Time|SetTimerbyEvent (delegate Custom|DelayedApply) 1.0))
          (Variables|Default|SetDelayApplyTimer _returnvalue_2))))))

(event Custom|DelayedApply
  (bind _returnvalue (CallFunction|GetAvatarasFortPlayerPawn))
  (GameplayEffects|RemoveActiveGameplayEffectbySourceEffect (Ability|GetAbilitySystemComponent (CallFunction|GetAvatarasFortPlayerPawn)) (Variables|Default|GetGE_Not_Equipped_CharacterPart) 0 -1)
  (Ability|ApplyGameplayEffectToOwner (Variables|Default|GetGE_Equipped_CharacterPart))
  (Class|FortCosmeticFunctionLibrary|ReselectGadgetCosmeticParts _returnvalue "/SevenSlide/Gameplay/AGID_BoostSevenSlider.AGID_BoostSevenSlider")
  (Class|FortCosmeticFunctionLibrary|ReselectGadgetCosmeticParts _returnvalue "/SevenSlide/Gameplay/AGID_Mythic_BoostSevenSlider.AGID_Mythic_BoostSevenSlider"))

(event Ability|EventOnEndAbility
  (params (bWasCancelled :type Boolean))
  (GameplayTasks|EndTask (Variables|Default|GetTogglePowerTask))
  (Utilities|Time|ClearandInvalidateTimerbyHandle (Variables|Default|GetDelayApplyTimer)))
